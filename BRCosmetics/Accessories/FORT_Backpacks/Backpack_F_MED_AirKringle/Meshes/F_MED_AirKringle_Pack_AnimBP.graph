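# Post-processes the AirKringle backpack pose with procedural secondary motion and character-state offsets.
# Anim Dynamics swings legs (two tunings blended by front-end state), wings and scarf, and springs the eyes.
# A Look At aims the pack at the body mesh's pelvis, cached each update, then additive offsets adjust legs for crouch-moving and DBNO, and pack for skydive floating.
# A deprecated function updates Gravity Override, but dynamics nodes use fixed gravity.
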AnimBlueprint /BRCosmetics/Accessories/FORT_Backpacks/Backpack_F_MED_AirKringle/Meshes/F_MED_AirKringle_Pack_AnimBP
target_skeleton=&/F_MED_AirKringle_Pack_Skeleton
graphs [
 AnimGraph {nodes [
   n1 = Anim Dynamics - Bone: dyn_leg_l!(ComponentPose=Local To Component(LocalPose=Input Pose() {name=InPose_0}), Alpha=1, GravityOverride:Vector="0,0,-600", ExternalForce="0,0,0") -> (Pose) {
    angularDampingOverride=0.96
    bOverrideAngularDamping=true
    bUseGravityOverride=true
    bUsePlanarLimit=false
    simulationSpace=Actor
    alphaBoolBlend {blendOption=Linear}
    boundBone {boneName=dyn_leg_l}
    componentAppliedLinearAccClamp=<5000 5000 5000>
    componentLinearAccScale=<0.15 0.225 0.125>
    componentLinearVelScale=<0.4 0.5 0.25>
    gravityOverride {z=-600}
    physicsBodyDefinitions [{
      boundBone {boneName=dyn_leg_l}
      boxExtents=<20 25 25>
      constraintSetup {angularLimitsMax=<5 30 20>, angularLimitsMin=<-5 -10 -60>, angularTarget=<0 0 0>}
      localJointOffset=<2 0 0>
     }]
   }
   n2 = Anim Dynamics - Bone: dyn_leg_r!(ComponentPose=@n1, Alpha=1, GravityOverride:Vector="0,0,-500", ExternalForce="0,0,50") -> (Pose) {
    angularDampingOverride=0.95
    angularSpringConstant=200
    bAngularSpring=true
    bOverrideAngularDamping=true
    bUseGravityOverride=true
    bUsePlanarLimit=false
    alphaBoolBlend {blendOption=Linear}
    boundBone {boneName=dyn_leg_r}
    componentAppliedLinearAccClamp=<8000 8000 8000>
    componentLinearAccScale=<0.2 0.4 0.2>
    componentLinearVelScale=<0.2 0.45 0.4>
    gravityOverride {z=-500}
    physicsBodyDefinitions [{
      boundBone {boneName=dyn_leg_r}
      boxExtents=<30 25 25>
      constraintSetup {angularLimitsMax=<20 15 25>, angularLimitsMin=<-20 -30 -60>, angularTarget=<0 0 0>}
      localJointOffset=<2.5 0.2 0>
     }]
   }
   n3 = Component To Local(ComponentPose=@n2)
   n4 = Anim Dynamics - Bone: dyn_leg_l(ComponentPose=Local To Component(LocalPose=Input Pose()), Alpha=1, GravityOverride="0,0,-600", ExternalForce="0,0,0") {
    angularDampingOverride=0.9
    bOverrideAngularDamping=true
    bUseGravityOverride=true
    bUsePlanarLimit=false
    alphaBoolBlend {blendOption=Linear}
    alphaScaleBiasClamp {bClampResult=true}
    boundBone {boneName=dyn_leg_l}
    componentAppliedLinearAccClamp=<5000 5000 7000>
    componentLinearAccScale=<0.3 0.45 0.25>
    componentLinearVelScale=<0.4 0.5 0.25>
    gravityOverride {z=-600}
    physicsBodyDefinitions [{
      boundBone {boneName=dyn_leg_l}
      boxExtents=<20 25 25>
      constraintSetup {angularLimitsMax=<5 30 20>, angularLimitsMin=<-5 -10 -60>, angularTarget=<0 0 0>}
      localJointOffset=<2 0 0>
     }]
   }
   n5 = Anim Dynamics - Bone: dyn_leg_r(ComponentPose=@n4, Alpha=1, GravityOverride="0,0,-500", ExternalForce="0,0,0") {
    angularDampingOverride=0.75
    angularSpringConstant=200
    bAngularSpring=true
    bOverrideAngularDamping=true
    bUseGravityOverride=true
    bUsePlanarLimit=false
    alphaBoolBlend {blendOption=Linear}
    alphaScaleBiasClamp {bClampResult=true}
    boundBone {boneName=dyn_leg_r}
    componentAppliedLinearAccClamp=<8000 8000 8000>
    componentLinearAccScale=<0.2 0.4 0.2>
    componentLinearVelScale=<0.2 0.45 0.4>
    gravityOverride {z=-500}
    physicsBodyDefinitions [{
      boundBone {boneName=dyn_leg_r}
      boxExtents=<30 25 25>
      constraintSetup {angularLimitsMax=<20 15 25>, angularLimitsMin=<-20 -30 -60>, angularTarget=<0 0 0>}
      localJointOffset=<2.5 0 0>
     }]
   }
   n6 = Component To Local(ComponentPose=@n5)
   n7 = Get bIsInFrontEnd()
   n8 = Anim Dynamics - Bone: dyn_wing_l!(
    ComponentPose=Local To Component(LocalPose=Blend Poses by bool(BlendPose_0=@n3, BlendPose_1=@n6, bActiveValue=@n7) {blendType=Linear})
    Alpha=1
    GravityOverride:Vector="0,0,-1000"
    ExternalForce="0,0,0"
   ) -> (Pose) {
    angularDampingOverride=0.9
    angularSpringConstant=100
    bAngularSpring=true
    bOverrideAngularDamping=true
    bUseGravityOverride=true
    alphaBoolBlend {blendOption=Linear}
    alphaScaleBiasClamp {bClampResult=true}
    boundBone {boneName=dyn_wing_l}
    componentAppliedLinearAccClamp=<1000 1000 1500>
    componentLinearAccScale=<0.2 0.25 0.6>
    componentLinearVelScale=<0.2 0.25 0.6>
    gravityOverride {z=-1000}
    physicsBodyDefinitions [{
      boundBone {boneName=dyn_wing_l}
      boxExtents=<13 10 10>
      constraintSetup {angularLimitsMax=<0 35 8>, angularLimitsMin=<0 -5 -8>, angularTarget=<1 0 0>}
      localJointOffset=<1 0 0>
     }]
   }
   n9 = Anim Dynamics - Bone: dyn_wing_r!(ComponentPose=@n8, Alpha=1, GravityOverride:Vector="0,0,-900", ExternalForce="0,0,0") -> (Pose) {
    angularDampingOverride=0.98
    angularSpringConstant=100
    bOverrideAngularDamping=true
    bUseGravityOverride=true
    alphaBoolBlend {blendOption=Linear}
    alphaScaleBiasClamp {bClampResult=true}
    boundBone {boneName=dyn_wing_r}
    componentAppliedLinearAccClamp=<1500 1500 1500>
    componentLinearAccScale=<0.2 0.25 0.8>
    componentLinearVelScale=<0.2 0.25 0.5>
    gravityOverride {z=-900}
    physicsBodyDefinitions [{
      boundBone {boneName=dyn_wing_r}
      boxExtents=<15 10 10>
      constraintSetup {angularLimitsMax=<0 5 8>, angularLimitsMin=<0 -35 -8>, angularTarget=<1 0 0>}
      localJointOffset=<1 0 0>
     }]
   }
   n10 = Anim Dynamics - Bone: dyn_scarf!(ComponentPose=@n9, Alpha=1, GravityOverride:Vector="0,0,-980", ExternalForce="0,0,0") -> (Pose) {
    angularDampingOverride=0.999
    angularSpringConstant=200
    bOverrideAngularDamping=true
    bUseGravityOverride=true
    alphaBoolBlend {blendOption=Linear}
    alphaScaleBiasClamp {bClampResult=true}
    boundBone {boneName=dyn_scarf}
    componentAppliedLinearAccClamp=<1000 1000 1000>
    componentLinearAccScale=<1 1 1>
    componentLinearVelScale=<1 1 1>
    gravityOverride {z=-980}
    physicsBodyDefinitions [{
      boundBone {boneName=dyn_scarf}
      boxExtents=<20 20 20>
      constraintSetup {angularLimitsMax=<0 0 45>, angularLimitsMin=<0 -90 -25>, angularTarget=<0 0 0>}
      localJointOffset=<2.5 0 0>
     }]
   }
   n11 = Anim Dynamics - Bone: dyn_eye_r(ComponentPose=@n10, Alpha=1, ExternalForce="0,0,0") {
    bLinearSpring=true
    bOverrideLinearDamping=true
    bUseGravityOverride=true
    linearDampingOverride=1
    linearSpringConstant=120
    alphaBoolBlend {blendOption=Linear}
    alphaScaleBiasClamp {bClampResult=true}
    boundBone {boneName=dyn_eye_r}
    componentAppliedLinearAccClamp=<800 800 800>
    componentLinearAccScale=<0.1 0.1 0.1>
    componentLinearVelScale=<0.1 0.1 0.1>
    gravityOverride {z=-200}
    physicsBodyDefinitions [{
      boundBone {boneName=dyn_eye_r}
      boxExtents=<12 12 12>
      constraintSetup {angularLimitsMax=<0 0 0>, angularLimitsMin=<0 0 0>, angularTarget=<0 0 0>, linearAxesMax=<0 0 1.3>, linearAxesMin=<0 0 -1.3>}
      localJointOffset=<0 0 -1.5>
     }]
   }
   n12 = Anim Dynamics - Bone: dyn_eye_l(ComponentPose=@n11, Alpha=1, ExternalForce="0,0,0") {
    bLinearSpring=true
    bOverrideLinearDamping=true
    bUseGravityOverride=true
    linearDampingOverride=1
    linearSpringConstant=150
    alphaBoolBlend {blendOption=Linear}
    alphaScaleBiasClamp {bClampResult=true}
    boundBone {boneName=dyn_eye_l}
    componentAppliedLinearAccClamp=<1000 1000 1000>
    componentLinearAccScale=<0.1 0.1 0.1>
    componentLinearVelScale=<0.1 0.1 0.1>
    gravityOverride {z=-300}
    physicsBodyDefinitions [{
      boundBone {boneName=dyn_eye_l}
      boxExtents=<15 15 15>
      constraintSetup {angularLimitsMax=<0 0 0>, angularLimitsMin=<0 0 0>, angularTarget=<0 0 0>, linearAxesMax=<0 0 1.3>, linearAxesMin=<0 0 -1.3>}
      localJointOffset=<0 0 -2>
     }]
   }
   n13 = Anim Dynamics - Bone: dyn_eye_r_base(ComponentPose=@n12, Alpha=1, ExternalForce="0,0,0") {
    angularDampingOverride=0.98
    angularSpringConstant=30
    bAngularSpring=true
    bOverrideAngularDamping=true
    bUseGravityOverride=true
    alphaBoolBlend {blendOption=Linear}
    alphaScaleBiasClamp {bClampResult=true}
    boundBone {boneName=dyn_eye_r_base}
    componentAppliedLinearAccClamp=<2000 2000 2000>
    componentLinearAccScale=<0.2 0.2 0.2>
    componentLinearVelScale=<0.2 0.2 0.2>
    gravityOverride {z=-100}
    physicsBodyDefinitions [{
      boundBone {boneName=dyn_eye_r_base}
      boxExtents=<10 10 10>
      constraintSetup {angularLimitsMax=<0 45 0>, angularLimitsMin=<0 -45 0>, angularTarget=<0 0 0>}
      localJointOffset=<0 0 -2>
     }]
   }
   n14 = Anim Dynamics - Bone: dyn_eye_l_base(ComponentPose=@n13, Alpha=1, ExternalForce="0,0,0") {
    angularDampingOverride=0.99
    angularSpringConstant=30
    bAngularSpring=true
    bOverrideAngularDamping=true
    bUseGravityOverride=true
    alphaBoolBlend {blendOption=Linear}
    alphaScaleBiasClamp {bClampResult=true}
    boundBone {boneName=dyn_eye_l_base}
    componentAppliedLinearAccClamp=<1500 1500 1500>
    componentLinearAccScale=<0.2 0.2 0.2>
    componentLinearVelScale=<0.2 0.2 0.2>
    gravityOverride {z=-100}
    physicsBodyDefinitions [{
      boundBone {boneName=dyn_eye_l_base}
      boxExtents=<10 10 10>
      constraintSetup {angularLimitsMax=<0 45 0>, angularLimitsMin=<0 -45 0>, angularTarget=<0 0 0>}
      localJointOffset=<0 0 -2>
     }]
   }
   n15 = Look At - Bone: pack!(ComponentPose=@n14, Alpha=1, LookAtLocation:Vector=Get PelvisLocation() -> (PelvisLocation:Vector)) -> (Pose) {
    alphaBoolBlend {blendOption=Linear}
    boneToModify {boneName=pack}
    lookAt_Axis {axis {y=0.4, z=-1}}
    lookUp_Axis {axis {x=0, z=1}}
    lookAtTarget=~
   }
   n16 = Get IsCrouchingMoving2D() -> (IsCrouchingMoving2D:bool)
   n17 = "Transform (Modify) Bone - Bone: dyn_leg_l_base"!(ComponentPose=@n15, bAlphaBoolEnabled:bool=@n16, Translation="0,0,0", Rotation="0,0,15", Scale="1,1,1") -> (Pose) {
    alphaInputType=Bool
    rotationMode=BMM_Additive
    alphaBoolBlend {blendInTime=0.2, blendOption=Linear, blendOutTime=0.8}
    boneToModify {boneName=dyn_leg_l_base}
    rotation {roll=15}
   }
   n18 = "Transform (Modify) Bone - Bone: dyn_leg_r_base"!(ComponentPose=@n17, bAlphaBoolEnabled:bool=@n16, Translation="0,0,0", Rotation="0,0,15", Scale="1,1,1") -> (Pose) {
    alphaInputType=Bool
    rotationMode=BMM_Additive
    alphaBoolBlend {blendInTime=0.2, blendOption=Linear, blendOutTime=0.8}
    boneToModify {boneName=dyn_leg_r_base}
    rotation {roll=15}
   }
   n19 = Get bIsDBNO()
   n20 = "Transform (Modify) Bone - Bone: dyn_leg_l_base"(ComponentPose=@n18, bAlphaBoolEnabled=@n19, Translation="0,0,0", Rotation="0,0,5", Scale="1,1,1") {
    alphaInputType=Bool
    rotationMode=BMM_Additive
    alphaBoolBlend {blendInTime=0.2, blendOption=Linear, blendOutTime=0.2}
    boneToModify {boneName=dyn_leg_l_base}
    rotation {roll=5}
   }
   n21 = "Transform (Modify) Bone - Bone: dyn_leg_r_base"(ComponentPose=@n20, bAlphaBoolEnabled=@n19, Translation="0,0,0", Rotation="0,0,10", Scale="1,1,1") {
    alphaInputType=Bool
    rotationMode=BMM_Additive
    alphaBoolBlend {blendInTime=0.2, blendOption=Linear, blendOutTime=0.2}
    boneToModify {boneName=dyn_leg_r_base}
    rotation {roll=10}
   }
   n22 = "Transform (Modify) Bone - Bone: pack"!(ComponentPose=@n21, bAlphaBoolEnabled:bool=Get bIsSkydiveFloating(), Translation="0,-5,-8", Rotation="0,0,2", Scale) -> (Pose) {
    alphaInputType=Bool
    rotationMode=BMM_Additive
    translationMode=BMM_Additive
    alphaBoolBlend {blendInTime=0.2, blendOption=Linear, blendOutTime=0.5}
    boneToModify {boneName=pack}
    rotation {roll=2}
    translation {y=-5, z=-8}
   }
   Output Pose(Result=Component To Local(ComponentPose=@n22)) {layerGroup=None}
  ]}
 EventGraph {nodes [
   Event BlueprintUpdateAnimation()
   n1 = Get bodyMesh() -> (bodyMesh:Skeletal Mesh Component Object Reference)
   > Is Valid(InputObject=@n1)
   > Set PelvisLocation(PelvisLocation:Vector=GetSocketLocation!(self:Scene Component Object Reference=@n1, InSocketName:Name=pelvis) -> (ReturnValue:Vector)) -> (Output_Get:Vector)
   > n2 = Gravity Override State Function DEPRECATED(CharacterPartAnimInstance=Self-Reference(), GameTimeInSeconds=GetGameTimeInSeconds(), GravityOverrideParams=Get Gravity Override Params())
   > Set Gravity Override(Gravity Override=@n2)
   > Set IsCrouchingMoving2D(IsCrouchingMoving2D:bool=AND Boolean(A=Get bIsCrouching(), B=Get bIsMoving2D())) -> (Output_Get:bool)
   n3 = Event OnCharacterCustomizationLoaded()
   > n4 = Cast To FortPlayerPawnAthena(Object=TryGetPawnOwner())
   > Set bodyMesh(bodyMesh:Skeletal Mesh Component Object Reference=GetSkeletalMeshForPartType(self=@n4)) -> (Output_Get:Skeletal Mesh Component Object Reference)
   Set bodyMesh(execute=@n4.CastFailed, bodyMesh=@n3.CustomBaseMesh)
   TryGetPawnOwner()
   Set bodyMesh(bodyMesh=@n3.CustomBaseMesh)
  ]}
]
variables [
 bodyMesh {type=SkeletalMeshComponent}
 Gravity Override {type=Vector}
 Gravity Override Params {type=GravityOverrideParamsStruct}
 IsCrouchingMoving2D {type=boolean}
 PelvisLocation {type=Vector}
]
